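; Test surface actor that displays a chosen physical surface type on cubes, apparently for testing impact effects per surface.
; The construction script maps the PhysSurface variable to a physical material through a large Select covering game, seasonal and default materials.
; It assigns that to a dynamic material instance created on Cube and shared with Cube1, then labels two text renders with the surface and material names.
; The event graph holds only empty events.
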(event EventBeginPlay)

(event Collision|EventActorBeginOverlap
  (params (OtherActor :type Actor%20Object%20Reference)))

(event EventTick
  (params (DeltaSeconds :type Float%20%28single-precision%29)))

(fn ConstructionScript
  (bind _physsurface (Variables|Default|GetPhysSurface))
  (bind _returnvalue (Utilities|Select :Index _physsurface :SurfaceType_Default "/Engine/EngineMaterials/DefaultPhysicalMaterial.DefaultPhysicalMaterial" :SurfaceType1 "/Game/Balance/Physics/Materials/PhysData_Wood.PhysData_Wood" :SurfaceType2 "/Game/Balance/Physics/Materials/PhysData_Stone.PhysData_Stone" :SurfaceType3 "/Game/FortressPhysicalMaterials/Metal.Metal" :SurfaceType4 "/Game/FortressPhysicalMaterials/Player_Flesh.Player_Flesh" :SurfaceType5 "/Game/FortressPhysicalMaterials/Player_Flesh.Player_Flesh" :SurfaceType6 "/Game/FortressPhysicalMaterials/Husk_Flesh.Husk_Flesh" :SurfaceType7 "/Game/FortressPhysicalMaterials/Purple_Flesh.Purple_Flesh" :SurfaceType8 "/Game/FortressPhysicalMaterials/Water.Water" :SurfaceType9 "/Game/FortressPhysicalMaterials/Grass.Grass" :SurfaceType10 "/Game/FortressPhysicalMaterials/Dirt.Dirt" :SurfaceType11 "/Game/FortressPhysicalMaterials/Explosive.Explosive" :SurfaceType12 "/Game/FortressPhysicalMaterials/WeakSpot.WeakSpot" :SurfaceType13 "/Engine/EngineMaterials/DefaultPhysicalMaterial.DefaultPhysicalMaterial" :SurfaceType14 "/Game/FortressPhysicalMaterials/WeakSpot_Wood.WeakSpot_Wood" :SurfaceType15 "/Game/FortressPhysicalMaterials/WeakSpot_Stone.WeakSpot_Stone" :SurfaceType16 "/Game/FortressPhysicalMaterials/WeakSpot_Metal.WeakSpot_Metal" :SurfaceType17 "/Game/FortressPhysicalMaterials/Player_Flesh.Player_Flesh" :SurfaceType18 "/Game/FortressPhysicalMaterials/Glass.Glass" :SurfaceType19 "/Game/FortressPhysicalMaterials/Husk_Flesh_Metal.Husk_Flesh_Metal" :SurfaceType20 "/Game/FortressPhysicalMaterials/Husk_Flesh_ice.Husk_Flesh_Ice" :SurfaceType21 "/Game/FortressPhysicalMaterials/Husk_Flesh_Electric.Husk_Flesh_Electric" :SurfaceType22 "/Game/FortressPhysicalMaterials/Husk_Flesh_Fire.Husk_Flesh_Fire" :SurfaceType23 "/Game/FortressPhysicalMaterials/Snow.Snow" :SurfaceType24 "/Game/FortressPhysicalMaterials/Ice.Ice" :SurfaceType25 "/Game/FortressPhysicalMaterials/Lava.Lava" :SurfaceType26 "/Game/FortressPhysicalMaterials/Sand.Sand" :SurfaceType27 "/Game/FortressPhysicalMaterials/Cube.Cube" :SurfaceType28 "/Game/TechArt/Seasonal/LackLuster/Physics/Chrome_PhysicalMaterial.Chrome_PhysicalMaterial" :SurfaceType29 "/Game/TechArt/Seasonal/LackLuster/Physics/Chrome_Petrified_PhysicalMaterial.Chrome_Petrified_PhysicalMaterial" :SurfaceType30 "/Game/Balance/Physics/Materials/PM_Mud.PM_Mud" :SurfaceType31 "/Game/Environments/Asteria/Sets/NeonCity/Materials/PhysMat_NeonCity_WetConcrete.PhysMat_NeonCity_WetConcrete" :SurfaceType32 "/Game/Balance/Physics/Materials/PM_TreeCanopy.PM_TreeCanopy" :SurfaceType33 "/Game/FortressPhysicalMaterials/Generic.Generic" :SurfaceType34 "/Engine/EngineMaterials/DefaultPhysicalMaterial.DefaultPhysicalMaterial"))
  (bind _returnvalue_2 (Rendering|Material|CreateDynamicMaterialInstance (Variables|Default|GetCube) 0 "/Game/Animation/Props/Galileo2/OLD/02_-_Default.02_-_Default" "None" :!class /Script/Engine.PrimitiveComponent))
  (Variables|/Script/Engine.MaterialInstance|SetPhysMaterial _returnvalue _returnvalue_2)
  (Rendering|Material|SetMaterial (Variables|Default|GetCube1) 0 _returnvalue_2)
  (Rendering|Components|TextRender|SetText (Variables|Default|GetTextRender) (Utilities|Text|ToText(Name) (Utilities|Name|EnumtoName _physsurface)))
  (Rendering|Components|TextRender|SetText (Variables|Default|GetTextRender1) (Utilities|Text|ToText(String) (Utilities|GetDisplayName _returnvalue))))
